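# One-shot stereo MetaSound base for the NitroDrome PA, layering an announcer voice line with two crowd reactions.
# On play, a BossSpawn comparison picks a random PA or NitroQueen line, while crowd layers A and B fire random sounds after random delays.
# PA and crowds mix to the output, but NitroQueen writes only to her own audio bus.
# OnFinished fires once both crowds and the chosen voice line end.
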MetaSoundSource /NitroDrome/Assets/Audio/PA/MSS_NitroDrome_PA_Base
inputs [
 BossSpawn {type=Bool}
 Crowd.A.DelayMax {type=Time, access=Constructor}
 Crowd.A.DelayMin {type=Time, access=Constructor}
 Crowd.A.Source {type=WaveAsset:Array, access=Constructor}
 Crowd.A.VolumeMultiplier {type=Float, access=Constructor}
 Crowd.B.DelayMax {type=Time, access=Constructor}
 Crowd.B.DelayMin {type=Time, access=Constructor}
 Crowd.B.Source {type=WaveAsset:Array, access=Constructor}
 Crowd.B.VolumeMultiplier {type=Float, access=Constructor}
 UE.Source.OnPlay {type=Trigger}
 VO.NitroQueen {type=WaveAsset:Array, access=Constructor}
 VO.NitroQueenVolume {type=Float, access=Constructor}
 VO.PA {type=WaveAsset:Array, access=Constructor}
 VO.PAVolume {type=Float, access=Constructor}
]
interfaces [UE.OutputFormat.Stereo:1.0, UE.Source.OneShot:1.0, UE.Source:1.1]
outputs [
 "UE.OutputFormat.Stereo.Audio:0" {type=Audio}
 "UE.OutputFormat.Stereo.Audio:1" {type=Audio}
 UE.Source.OneShot.OnFinished {type=Trigger}
]
pages [Default {
  input_defaults {
   BossSpawn=null
   Crowd.A.DelayMax=2
   Crowd.A.DelayMin=2
   Crowd.A.VolumeMultiplier=0.6
   Crowd.B.DelayMax=0
   Crowd.B.DelayMin=0
   Crowd.B.VolumeMultiplier=0.7
   UE.Source.OnPlay=null
   VO.NitroQueenVolume=1.5
   VO.PAVolume=0.85
   Crowd.A.Source []
   Crowd.B.Source []
   VO.NitroQueen []
   VO.PA []
  }
  nodes [
   n1 = InitVariable.Trigger()
   n2 = TriggerCompare.Clamp:Bool(A=^BossSpawn, Compare=^UE.Source.OnPlay) {also {defaults {B="<MetaSoundBuilderResult.SUCCEEDED: 0>"}}}
   n3 = &MSS_PlayRandomSound_Stereo(
    Sounds=^VO.NitroQueen
    MaxVolume=^VO.NitroQueenVolume
    MinVolume=^VO.NitroQueenVolume
    UE.Source.OnPlay=VariableAccessor.Trigger(Variable=VariableMutator.Trigger(Variable=@n1.Variable, Value=@n2.True).Variable).Value
   )
   AudioBusWriter2_0 = "UE.Audio Bus Writer (2)"(In 0=@n3."UE.OutputFormat.Stereo.Audio:0", In 1=@n3."UE.OutputFormat.Stereo.Audio:1") {also {defaults {Audio Bus=/Game/Sounds/Helios/POI/NitroDrome/NitroQueen/NitroQueen_AudioBus}}}
   n4 = UE.RandomTime(Max=^Crowd.B.DelayMax, Min=^Crowd.B.DelayMin, Next=^UE.Source.OnPlay)
   TriggerDelay_1 = UE.Trigger Delay(In=@n4.On Next, Delay Time=@n4.Value)
   n5 = &MSS_PlayRandomSound_Stereo(Sounds=^Crowd.B.Source, MaxVolume=^Crowd.B.VolumeMultiplier, MinVolume=^Crowd.B.VolumeMultiplier, UE.Source.OnPlay=@TriggerDelay_1.Out)
   n6 = InitVariable.Trigger()
   n7 = &MSS_PlayRandomSound_Stereo(
    Sounds=^VO.PA
    MaxVolume=^VO.PAVolume
    MinVolume=^VO.PAVolume
    UE.Source.OnPlay=VariableAccessor.Trigger(Variable=VariableMutator.Trigger(Variable=@n6.Variable, Value=@n2.False).Variable).Value
   )
   n8 = UE.RandomTime(Max=^Crowd.A.DelayMax, Min=^Crowd.A.DelayMin, Next=^UE.Source.OnPlay)
   TriggerDelay_0 = UE.Trigger Delay(In=@n8.On Next, Delay Time=@n8.Value)
   n9 = &MSS_PlayRandomSound_Stereo(Sounds=^Crowd.A.Source, MaxVolume=^Crowd.A.VolumeMultiplier, MinVolume=^Crowd.A.VolumeMultiplier, UE.Source.OnPlay=@TriggerDelay_0.Out)
   AudioMixerAudioMixerStereo3_0 = "AudioMixer.Audio Mixer (Stereo, 3)"(
    In 1 L=@n5."UE.OutputFormat.Stereo.Audio:0"
    In 1 R=@n5."UE.OutputFormat.Stereo.Audio:1"
    In 2 L=@n7."UE.OutputFormat.Stereo.Audio:0"
    In 2 R=@n7."UE.OutputFormat.Stereo.Audio:1"
    In 0 L=@n9."UE.OutputFormat.Stereo.Audio:0"
    In 0 R=@n9."UE.OutputFormat.Stereo.Audio:1"
   )
   n10 = InitVariable.Trigger()
   n11 = InitVariable.Trigger()
   n12 = InitVariable.Trigger()
   TriggerAnyTriggerAny2_0 = "TriggerAny.Trigger Any (2)"(In 1=@n3.UE.Source.OneShot.OnFinished, In 0=@n7.UE.Source.OneShot.OnFinished)
   TriggerAccumulatorTriggerAccumulate3_0 = "TriggerAccumulator.Trigger Accumulate (3)"(
    In 1=VariableAccessor.Trigger(Variable=VariableMutator.Trigger(Variable=@n10.Variable, Value=@n5.UE.Source.OneShot.OnFinished).Variable).Value
    In 0=VariableAccessor.Trigger(Variable=VariableMutator.Trigger(Variable=@n11.Variable, Value=@n9.UE.Source.OneShot.OnFinished).Variable).Value
    In 2=VariableAccessor.Trigger(Variable=VariableMutator.Trigger(Variable=@n12.Variable, Value=@TriggerAnyTriggerAny2_0.Out).Variable).Value
   )
  ]
  links [
   {dst="@UE.OutputFormat.Stereo.Audio:0", src=AudioMixerAudioMixerStereo3_0.Out L}
   {dst="@UE.OutputFormat.Stereo.Audio:1", src=AudioMixerAudioMixerStereo3_0.Out R}
   {dst="@UE.Source.OneShot.OnFinished", src=TriggerAccumulatorTriggerAccumulate3_0.Out}
  ]
 }]
properties {
 assetImportData=&self:AssetImportData
 attenuationSettings=/NitroDrome/Assets/Audio/PA/VO/VO_PA_Announcer_Att
 maxDistance=20000
 numChannels=2
 outputFormat=Stereo
 priority=1.25
 sampleRate=48000
 soundClassObject=&Fort_VO_Enemy_Critical
 concurrencySet [/NitroDrome/Assets/Audio/PA/VO/VO_PA_Announcer_Conc]
}
