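# Stereo one-shot MetaSound preset for the NitroDrome PA, named for the Nitro Queen boss entering the fight.
# It overrides MSS_NitroDrome_PA_Base inputs with two delayed crowd cheer layers and Nitro Queen voice lines, leaving playback logic to the parent.
# The PA announcer array is empty, and the Queen's array uses Player_Elim clips despite the asset's name.
# Announcer attenuation, critical enemy VO class and two concurrency sets govern playback.
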MetaSoundSource /NitroDrome/Assets/Audio/PA/Metasounds/MSS_Nitro_Queen_Boss_Enters_Fight
preset_parent=/NitroDrome/Assets/Audio/PA/MSS_NitroDrome_PA_Base
inputs [
 BossSpawn {type=Bool}
 Crowd.A.DelayMax {type=Time, access=Constructor}
 Crowd.A.DelayMin {type=Time, access=Constructor}
 Crowd.A.Source {type=WaveAsset:Array, access=Constructor}
 Crowd.A.VolumeMultiplier {type=Float, access=Constructor}
 Crowd.B.DelayMax {type=Time, access=Constructor}
 Crowd.B.DelayMin {type=Time, access=Constructor}
 Crowd.B.Source {type=WaveAsset:Array, access=Constructor}
 Crowd.B.VolumeMultiplier {type=Float, access=Constructor}
 UE.Source.OnPlay {type=Trigger}
 VO.NitroQueen {type=WaveAsset:Array, access=Constructor}
 VO.NitroQueenVolume {type=Float, access=Constructor}
 VO.PA {type=WaveAsset:Array, access=Constructor}
 VO.PAVolume {type=Float, access=Constructor}
]
interfaces [UE.OutputFormat.Stereo:1.0, UE.Source.OneShot:1.0, UE.Source:1.1]
outputs [
 "UE.OutputFormat.Stereo.Audio:0" {type=Audio}
 "UE.OutputFormat.Stereo.Audio:1" {type=Audio}
 UE.Source.OneShot.OnFinished {type=Trigger}
]
pages [Default {input_defaults {
   BossSpawn=null
   Crowd.A.DelayMax=2
   Crowd.A.DelayMin=2
   Crowd.A.VolumeMultiplier=0.6
   Crowd.B.DelayMax=3.5
   Crowd.B.DelayMin=2.75
   Crowd.B.VolumeMultiplier=0.7
   UE.Source.OnPlay=null
   VO.NitroQueenVolume=1.5
   VO.PAVolume=0.85
   Crowd.A.Source [
    /NitroDrome/Assets/Audio/PA/Crowd/NitroDrome_PA_Crowd_Cheer_01
    /NitroDrome/Assets/Audio/PA/Crowd/NitroDrome_PA_Crowd_Cheer_02
    /NitroDrome/Assets/Audio/PA/Crowd/NitroDrome_PA_Crowd_Cheer_03
    /NitroDrome/Assets/Audio/PA/Crowd/NitroDrome_PA_Crowd_Cheer_04
   ]
   Crowd.B.Source [
    /NitroDrome/Assets/Audio/PA/Crowd/NitroDrome_PA_Crowd_Cheer_08
    /NitroDrome/Assets/Audio/PA/Crowd/NitroDrome_PA_Crowd_Cheer_07
    /NitroDrome/Assets/Audio/PA/Crowd/NitroDrome_PA_Crowd_Cheer_06
    /NitroDrome/Assets/Audio/PA/Crowd/NitroDrome_PA_Crowd_Cheer_05
   ]
   VO.NitroQueen [
    /NitroDrome/Assets/Audio/VO/Nitro_Queen/VO_Nitro_Queen_Player_Elim_01
    /NitroDrome/Assets/Audio/VO/Nitro_Queen/VO_Nitro_Queen_Player_Elim_02
    /NitroDrome/Assets/Audio/VO/Nitro_Queen/VO_Nitro_Queen_Player_Elim_03
    /NitroDrome/Assets/Audio/VO/Nitro_Queen/VO_Nitro_Queen_Player_Elim_04
    /NitroDrome/Assets/Audio/VO/Nitro_Queen/VO_Nitro_Queen_Player_Elim_05
    /NitroDrome/Assets/Audio/VO/Nitro_Queen/VO_Nitro_Queen_Player_Elim_06
    /NitroDrome/Assets/Audio/VO/Nitro_Queen/VO_Nitro_Queen_Player_Elim_07
   ]
   VO.PA []
  }}]
properties {
 assetImportData=&self:AssetImportData
 attenuationSettings=/NitroDrome/Assets/Audio/PA/VO/VO_PA_Announcer_Att
 maxDistance=20000
 numChannels=2
 outputFormat=Stereo
 priority=1.05
 sampleRate=48000
 soundClassObject=&Fort_VO_Enemy_Critical
 concurrencySet [/NitroDrome/Assets/Audio/PA/VO/VO_PA_Announcer_Conc, /Game/Sounds/Helios/POI/NitroDrome/NitroQueen/VO_NitroQueen_PreBattle_Conc]
}
